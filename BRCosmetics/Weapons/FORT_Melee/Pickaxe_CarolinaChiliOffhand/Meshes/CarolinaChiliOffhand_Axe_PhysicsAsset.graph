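# Physics asset for the Carolina Chili offhand pickaxe cosmetic, a melee axe mesh.
# It approximates the weapon with a single capsule body on the Melee_TwoHanded_Handle bone, offset and tilted from that bone.
# It has no constraints, so no articulated or jointed parts are simulated.
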PhysicsAsset /BRCosmetics/Weapons/FORT_Melee/Pickaxe_CarolinaChiliOffhand/Meshes/CarolinaChiliOffhand_Axe_PhysicsAsset
bodies {Melee_TwoHanded_Handle [{shapeType=Capsule, length=22.4763, radius=7.2868, center=<-8.8331 0.4461 -1.2024>, rotation=<rot 0 -9.2842 -78.8864>}]}
constraints []
